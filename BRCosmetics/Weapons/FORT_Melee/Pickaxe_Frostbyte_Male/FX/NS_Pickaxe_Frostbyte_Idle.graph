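# Niagara idle effect for the Frostbyte pickaxe, shedding frost mist and snowflakes from points sampled on its skeletal mesh.
# Two Frost sprite emitters raise spawn rate and shorten lifetime with owner speed, fading by distance and brightening alpha when FrontendSpawn is set.
# Their positions blend from owner-local toward simulated motion over age, so mist clings before trailing.
# Snowflakes spawn steadily with curl noise; Frost002 is disabled.
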NiagaraSystem /BRCosmetics/Weapons/FORT_Melee/Pickaxe_Frostbyte_Male/FX/NS_Pickaxe_Frostbyte_Idle
dependencies {
 usedDataInterfaces [
  /Script/Niagara.NiagaraDataInterfaceSimpleCounter
  &NiagaraDataInterfaceCurve
  &NiagaraDataInterfaceVectorCurve
  &NiagaraDataInterfaceColorCurve
 ]
 usedDynamicInputs [&ModuloInt, /Niagara/DynamicInputs/DataInterface/GenerateIntFromCounter]
 usedModules [&SystemState]
}
emitters {
 Frost {
  emitterUpdateScript [
   &EmitterState()
   /SetVariables(Emitter.PlayerVelocityVector:Vector3f=$Engine.Owner.Velocity) {isSetParameters=true, module=SetVariables_#1}
   /SetVariables(Emitter.PlayerSpeed:NiagaraFloat=&VectorLength(Vector=$Emitter.PlayerVelocityVector)) {isSetParameters=true, module=SetVariables_#2}
   &SpawnRate(
    SpawnRate=&Lerp_Float(A=3, B=120, Alpha=&ClampFloat(Float=&Divide_Float(A=$Emitter.PlayerSpeed, B=100)))
    Use Spawn Probability=false
   )
  ]
  particleSpawnScript [
   /Niagara/Modules/Spawn/Initialization/InitializeParticle(
    Color=&MakeCustomLinearColorFromBool(Bool=$User.FrontendSpawn, True Color=<rgba 0.2779 0.6803 1 0.7>, False Color=<rgba 0.2779 0.6803 1 0.3>)
    Lifetime=&Multiply_Float(
     A=&RandomRangeFloat(Minimum=0.75, Maximum=1)
     B=&Lerp_Float(
      A=2
      B=&RandomRangeFloat(Minimum=0.6, Maximum=0.3)
      Alpha=&ClampFloat(Float=&Divide_Float(A=$Emitter.PlayerSpeed, B=100))
     )
    )
    Module.Write Color=true
    Module.Write Lifetime=true
    Module.Write Mass=true
    Module.Write Material Random=true
    Module.Write Position=true
    Module.Write Scale=false
    Module.Write SpriteRotation=true
    Module.Write SpriteSize=true
    Sprite Rotation=&UniformAOrBFloat(A=0, B=180)
    Sprite Size=&Vector2DFromFloat(Value=&RandomRangeFloat(Minimum=8, Maximum=10))
   )
   &SkeletalMeshLocation(
    Inherit Velocity=false
    MaskSpawnGroup=false
    Mesh Sampling Type=::NewEnumerator5
    Sampled Position Offset=&RandomRangeVector(Minimum=<-5 0 0>, Maximum=<5 0 0>)
    Sampled Position Offset Coordinate Space:ENiagaraMeshTransforms=::NewEnumerator3
    Skeletal Mesh=$User.SkeletalMesh
    Socket Sampling Mode=::NewEnumerator5
   )
   &InheritVelocity(Module.Limit Speed=true, Velocity Scale=&Multiply_VectorByFloat(Vector=<1 1 1>, Float=100)) {enabled=false}
   /SetVariables(Particles.PickaxeLocation:Vector3f=$Output.SkeletalMeshLocation.SampledPosition) {isSetParameters=true, module=SetVariables_#3}
   /SetVariables(Particles.LocalPosition:Vector3f=/Niagara/DynamicInputs/Transforms/TransformPosition(Position=Linked() {linkedVariable {name=Particles.Position, type=NiagaraPosition}}, Source Space=::World, Destination Space=::Local)) {isSetParameters=true, module=SetVariables_#4}
  ]
  particleUpdateScript [
   &ParticleState()
   &SubUVAnimation(End Frame=40, Start Frame=2)
   &AddVelocity(Velocity=<0 -0.25 0>) {enabled=false}
   &GravityForce(Gravity=&RandomRangeVector(Minimum=<0 0 -10>, Maximum=<0 0 -20>))
   &InheritVelocity(Module.Limit Speed=true, Velocity Scale=&Multiply_VectorByFloat(Vector=<1 1 1>, Float=0.001)) {enabled=false, module=InheritVelocity001}
   &SolveForcesAndVelocity(Clamp Velocity=false, Limit Acceleration=false)
   &ScaleColor(
    Scale Alpha=&FloatFromCurve(
     FloatCurve=DataInterface() {propertyValues {Curve {keys [
         {arriveTangent=0.1402, interpMode=RCIM_Cubic, leaveTangent=0.1402, tangentMode=RCTM_User, time=0, value=0}
         {arriveTangent=0.317, interpMode=RCIM_Cubic, leaveTangent=0.317, tangentMode=RCTM_User, time=0.3, value=0.35}
         <at 0.99 0 RCIM_Cubic>
        ]}}}
     Scale Curve=&FloatFromCurve(
      FloatCurve=DataInterface() {propertyValues {Curve {keys [<at 0.25 1 RCIM_Cubic>,<at 1 0 RCIM_Cubic>]}}}
      CurveIndex=&NormalizeDistanceRange(
       Start Position:NiagaraPosition=Linked() {linkedVariable {name=Particles.PickaxeLocation, type=Vector3f}}
       End Position:NiagaraPosition=$Particles.Position
       Distance:NiagaraFloat=15
       Invert Normalized Range:NiagaraBool=false
      )
     )
    )
    ScaleA=true
    ScaleRGB=true
   )
   &CameraOffset(Camera Offset Amount=&FloatFromCurve(FloatCurve=DataInterface() {propertyValues {Curve {keys [<at 0 1>,<at 1 20>]}}}))
   &ScaleSpriteSize(Scale Factor=&Multiply_Vector2(
     A=<1 1>
     B=&Vector2DFromFloat(Value=&FloatFromCurve(FloatCurve=DataInterface() {propertyValues {Curve {keys [<at 0.3 1>,<at 1 2>]}}}))
    ))
   &DynamicMaterialParameters(
    Index 0 Param 4=1
    Module.Param0WriteEnabled=true
    Module.Param1WriteEnabled=true
    Module.Param2WriteEnabled=true
    Module.Param3WriteEnabled=true
   )
   /SetVariables(Particles.Position=&Lerp_Vector(
     A=&TransformPosition(Position=$Particles.LocalPosition, Source Space=::Local, Destination Space=::Simulation)
     B=Linked() {linkedVariable {name=Particles.Position, type=NiagaraPosition}}
     Alpha=&FloatFromCurve(FloatCurve=DataInterface() {propertyValues {Curve {keys [
          {arriveTangent=2.9246, interpMode=RCIM_Cubic, leaveTangent=2.9247, tangentMode=RCTM_User, time=0, value=0}
          {arriveTangent=0.0307, interpMode=RCIM_Cubic, tangentMode=RCTM_Break, time=0.6, value=1}
         ]}}})
    )) {isSetParameters=true, module=SetVariables_#5}
  ]
  rendererClasses [&NiagaraSpriteRendererProperties]
  renderers [{rendererClass=&NiagaraSpriteRendererProperties}]
 }
 Frost001 {
  emitterUpdateScript [
   &EmitterState()
   /SetVariables(Emitter.PlayerVelocityVector:Vector3f=$Engine.Owner.Velocity) {isSetParameters=true, module=SetVariables_#1}
   /SetVariables(Emitter.PlayerSpeed:NiagaraFloat=&VectorLength(Vector=$Emitter.PlayerVelocityVector)) {isSetParameters=true, module=SetVariables_#2}
   &SpawnRate(
    SpawnRate=&Lerp_Float(A=3, B=20, Alpha=&ClampFloat(Float=&Divide_Float(A=$Emitter.PlayerSpeed, B=100)))
    Use Spawn Probability=false
   )
  ]
  particleSpawnScript [
   &InitializeParticle(
    Color=&MakeCustomLinearColorFromBool(Bool=$User.FrontendSpawn, True Color=<rgba 0.2779 0.6803 1 0.7>, False Color=<rgba 0.2779 0.6803 1 0.3>)
    Lifetime=&Multiply_Float(
     A=&RandomRangeFloat(Minimum=0.75, Maximum=1)
     B=&Lerp_Float(
      A=2
      B=&RandomRangeFloat(Minimum=0.5, Maximum=0.3)
      Alpha=&ClampFloat(Float=&Divide_Float(A=$Emitter.PlayerSpeed, B=250))
     )
    )
    Module.Write Color=true
    Module.Write Lifetime=true
    Module.Write Mass=true
    Module.Write Material Random=true
    Module.Write Position=true
    Module.Write Scale=false
    Module.Write SpriteRotation=true
    Module.Write SpriteSize=true
    Sprite Rotation=&UniformAOrBFloat(A=0, B=180)
    Sprite Size=&Vector2DFromFloat(Value=&RandomRangeFloat(Minimum=10, Maximum=16))
   )
   &SkeletalMeshLocation(
    Inherit Velocity=false
    MaskSpawnGroup=false
    Mesh Sampling Type=::NewEnumerator5
    Sampled Position Offset=&RandomRangeVector(Minimum=<-5 0 0>, Maximum=<5 0 0>)
    Sampled Position Offset Coordinate Space:ENiagaraMeshTransforms=::NewEnumerator3
    Skeletal Mesh=$User.SkeletalMesh
    Socket Sampling Mode=::NewEnumerator5
   )
   /SetVariables(Particles.PickaxeLocation:Vector3f=$Output.SkeletalMeshLocation.SampledPosition) {isSetParameters=true, module=SetVariables_#3}
   /SetVariables(Particles.LocalPosition:Vector3f=&TransformPosition(Position=Linked() {linkedVariable {name=Particles.Position, type=NiagaraPosition}}, Source Space=::World, Destination Space=::Local)) {isSetParameters=true, module=SetVariables_#4}
  ]
  particleUpdateScript [
   &ParticleState()
   &SubUVAnimation(End Frame=40, Start Frame=2)
   &SkeletalMeshLocation(
    Apply To Particle Position=false
    Inherit Velocity=false
    MaskSpawnGroup=false
    Mesh Sampling Type=::NewEnumerator5
    Sampled Position Offset=&RandomRangeVector(Minimum=<-5 0 0>, Maximum=<5 0 0>)
    Sampled Position Offset Coordinate Space:ENiagaraMeshTransforms=::NewEnumerator3
    Skeletal Mesh=$User.SkeletalMesh
    Socket Sampling Mode=::NewEnumerator5
   ) {module=SkeletalMeshLocation001}
   /NiagaraFN/Sprite/FN_SpriteAlignmentAndFacing(SpriteAlignment:Vector3f=&Subtract_Vector(
     A=&Add_Vector(A=Linked() {linkedVariable {name=Particles.Position, type=NiagaraPosition}}, B=<0 0 -10>)
     B=$Output.SkeletalMeshLocation001.SampledPosition
    ))
   &AddVelocity(Velocity=<0 -0.25 0>) {enabled=false}
   &GravityForce(Gravity=&Multiply_Vector(
     A=&RandomRangeVector(Minimum=<0 0 -10>, Maximum=<0 0 -20>)
     B=&VectorFromFloat(Value=25)
    ))
   &SolveForcesAndVelocity(Clamp Velocity=false, Limit Acceleration=false)
   &ScaleColor(
    Scale Alpha=&FloatFromCurve(
     FloatCurve=DataInterface() {propertyValues {Curve {keys [
         {arriveTangent=0.1402, interpMode=RCIM_Cubic, leaveTangent=0.1402, tangentMode=RCTM_User, time=0, value=0}
         {arriveTangent=0.317, interpMode=RCIM_Cubic, leaveTangent=0.317, tangentMode=RCTM_User, time=0.3, value=0.35}
         <at 0.99 0 RCIM_Cubic>
        ]}}}
     Scale Curve=&FloatFromCurve(
      FloatCurve=DataInterface() {propertyValues {Curve {keys [<at 0.25 1 RCIM_Cubic>,<at 1 0 RCIM_Cubic>]}}}
      CurveIndex=&ClampFloat(Float=&Multiply_Float(
        A=&VectorLength(Vector=&Subtract_Vector(A=$Output.SkeletalMeshLocation001.SampledPosition, B=Linked() {linkedVariable {name=Particles.Position, type=NiagaraPosition}}))
        B=0.025
       ))
     )
    )
    ScaleA=true
    ScaleRGB=true
   )
   &CameraOffset(Camera Offset Amount=&FloatFromCurve(FloatCurve=DataInterface() {propertyValues {Curve {keys [<at 0 1>,<at 1 20>]}}}))
   &ScaleSpriteSize(Scale Factor=/Niagara/DynamicInputs/ScaleByCurve/ScaleVector2DByCurve(
     InitialValue:Vector2f=<1 1>
     ScaleCurve:NiagaraDataInterfaceVectorCurve=DataInterface() {propertyValues {
       XCurve {keys [<at 0 1>, {arriveTangent=0.2843, interpMode=RCIM_Cubic, leaveTangent=0.2843, tangentMode=RCTM_User, time=0.0815, value=1}, <at 1 1.5>]}
       YCurve {keys [
         <at 0 1 RCIM_Cubic>
         {arriveTangent=1.0179, interpMode=RCIM_Cubic, leaveTangent=1.0179, tangentMode=RCTM_User, time=0.2983, value=1.0887}
         <at 1 2.5 RCIM_Cubic>
        ]}
       ZCurve {keys [<at 0.0788 0>]}
      }}
     CurveIndex:NiagaraFloat=$Particles.NormalizedAge
    ))
   &DynamicMaterialParameters(
    Index 0 Param 4=1
    Module.Param0WriteEnabled=true
    Module.Param1WriteEnabled=true
    Module.Param2WriteEnabled=true
    Module.Param3WriteEnabled=true
   )
   /SetVariables(Particles.Position=&Lerp_Vector(
     A=&TransformPosition(Position=$Particles.LocalPosition, Source Space=::Local, Destination Space=::Simulation)
     B=Linked() {linkedVariable {name=Particles.Position, type=NiagaraPosition}}
     Alpha=&FloatFromCurve(FloatCurve=DataInterface() {propertyValues {Curve {keys [
          {arriveTangent=0.3418, interpMode=RCIM_Cubic, leaveTangent=0.3418, tangentMode=RCTM_User, time=0, value=0}
          {arriveTangent=0.0307, interpMode=RCIM_Cubic, tangentMode=RCTM_Break, time=0.6, value=0.5}
         ]}}})
    )) {isSetParameters=true, module=SetVariables_#5}
   &Color(Color=&ColorFromCurve(
     ColorCurve=DataInterface() {propertyValues {
       AlphaCurve {keys [<at 0 1 RCIM_Cubic>]}
       BlueCurve {keys [<at 0 0 RCIM_Cubic>,<at 0.5 0>,<at 1 1 RCIM_Cubic>]}
       GreenCurve {keys [<at 0 0 RCIM_Cubic>,<at 0.5 1>,<at 1 0 RCIM_Cubic>]}
       RedCurve {keys [<at 0 1 RCIM_Cubic>,<at 0.5 0.2094>,<at 1 0.0831 RCIM_Cubic>]}
      }}
     CurveIndex=&ClampFloat(Float=&Divide_Float(A=$Emitter.PlayerSpeed, B=1000))
    )) {enabled=false}
  ]
  rendererClasses [&NiagaraSpriteRendererProperties]
  renderers [{rendererClass=&NiagaraSpriteRendererProperties}]
 }
 Frost002 {
  bEnabled=false
  emitterUpdateScript [
   &EmitterState()
   /SetVariables(Emitter.PlayerVelocityVector:Vector3f=$Engine.Owner.Velocity) {isSetParameters=true, module=SetVariables_#1}
   /SetVariables(Emitter.PlayerSpeed:NiagaraFloat=&VectorLength(Vector=$Emitter.PlayerVelocityVector)) {isSetParameters=true, module=SetVariables_#2}
   &SpawnRate(
    SpawnRate=&Lerp_Float(
     A=6
     B=0
     Alpha=&ClampFloat(Float=&Subtract_Float(A=&Divide_Float(A=$Emitter.PlayerSpeed, B=10), B=10))
    )
    Use Spawn Probability=false
   )
  ]
  particleSpawnScript [
   &InitializeParticle(
    Color=&MakeCustomLinearColorFromBool(Bool=$User.FrontendSpawn, True Color=<rgba 0.2779 0.6803 1 0.7>, False Color=<rgba 0.2779 0.6803 1 0.3>)
    Lifetime=&Multiply_Float(
     A=&RandomRangeFloat(Minimum=0.75, Maximum=1)
     B=&Lerp_Float(
      A=2
      B=&RandomRangeFloat(Minimum=0.3, Maximum=0.1)
      Alpha=&ClampFloat(Float=&Divide_Float(A=$Emitter.PlayerSpeed, B=100))
     )
    )
    Module.Write Color=true
    Module.Write Lifetime=true
    Module.Write Mass=true
    Module.Write Material Random=true
    Module.Write Position=true
    Module.Write Scale=false
    Module.Write SpriteRotation=true
    Module.Write SpriteSize=true
    Sprite Rotation=&UniformAOrBFloat(A=0, B=180)
    Sprite Size=&Vector2DFromFloat(Value=&RandomRangeFloat(Minimum=8, Maximum=10))
   )
   &SkeletalMeshLocation(
    Inherit Velocity=false
    MaskSpawnGroup=false
    Mesh Sampling Type=::NewEnumerator5
    Sampled Position Offset=&RandomRangeVector(Minimum=<-5 0 0>, Maximum=<5 0 0>)
    Sampled Position Offset Coordinate Space:ENiagaraMeshTransforms=::NewEnumerator3
    Skeletal Mesh=$User.SkeletalMesh
    Socket Sampling Mode=::NewEnumerator5
   )
   /SetVariables(Particles.PickaxeLocation:Vector3f=$Output.SkeletalMeshLocation.SampledPosition) {isSetParameters=true, module=SetVariables_#3}
   /SetVariables(Particles.LocalPosition:Vector3f=&TransformPosition(Position=Linked() {linkedVariable {name=Particles.Position, type=NiagaraPosition}}, Source Space=::World, Destination Space=::Local)) {isSetParameters=true, module=SetVariables_#4}
  ]
  particleUpdateScript [
   &ParticleState()
   &SubUVAnimation(End Frame=40, Start Frame=2)
   &SkeletalMeshLocation(
    Apply To Particle Position=false
    Inherit Velocity=false
    MaskSpawnGroup=false
    Mesh Sampling Type=::NewEnumerator5
    Sampled Position Offset=&RandomRangeVector(Minimum=<-5 0 0>, Maximum=<5 0 0>)
    Sampled Position Offset Coordinate Space:ENiagaraMeshTransforms=::NewEnumerator3
    Skeletal Mesh=$User.SkeletalMesh
    Socket Sampling Mode=::NewEnumerator5
   ) {module=SkeletalMeshLocation001}
   &FN_SpriteAlignmentAndFacing(SpriteAlignment:Vector3f=&Subtract_Vector(
     A=&Add_Vector(A=Linked() {linkedVariable {name=Particles.Position, type=NiagaraPosition}}, B=<0 0 -5>)
     B=$Output.SkeletalMeshLocation001.SampledPosition
    ))
   &AddVelocity(Velocity=<0 -0.25 0>) {enabled=false}
   &GravityForce(Gravity=&Multiply_Vector(
     A=&RandomRangeVector(Minimum=<0 0 -10>, Maximum=<1 1 -20>)
     B=&VectorFromFloat(Value=25)
    ))
   &SolveForcesAndVelocity(Clamp Velocity=false, Limit Acceleration=false)
   &ScaleColor(
    Scale Alpha=&FloatFromCurve(
     FloatCurve=DataInterface() {propertyValues {Curve {keys [
         {arriveTangent=0.1402, interpMode=RCIM_Cubic, leaveTangent=0.1402, tangentMode=RCTM_User, time=0, value=0}
         {arriveTangent=0.317, interpMode=RCIM_Cubic, leaveTangent=0.317, tangentMode=RCTM_User, time=0.3, value=0.35}
         <at 0.99 0 RCIM_Cubic>
        ]}}}
     Scale Curve=&FloatFromCurve(
      FloatCurve=DataInterface() {propertyValues {Curve {keys [<at 0.25 1 RCIM_Cubic>,<at 1 0 RCIM_Cubic>]}}}
      CurveIndex=&ClampFloat(Float=&Multiply_Float(
        A=&VectorLength(Vector=&Subtract_Vector(A=$Output.SkeletalMeshLocation001.SampledPosition, B=Linked() {linkedVariable {name=Particles.Position, type=NiagaraPosition}}))
        B=0.025
       ))
     )
    )
    ScaleA=true
    ScaleRGB=true
   )
   &CameraOffset(Camera Offset Amount=&FloatFromCurve(FloatCurve=DataInterface() {propertyValues {Curve {keys [<at 0 1>,<at 1 20>]}}}))
   &ScaleSpriteSize(Scale Factor=&ScaleVector2DByCurve(
     InitialValue:Vector2f=<1 1>
     ScaleCurve:NiagaraDataInterfaceVectorCurve=DataInterface() {propertyValues {
       XCurve {keys [<at 0.0788 1>,<at 0.0815 1>]}
       YCurve {keys [{arriveTangent=1.8253, interpMode=RCIM_Cubic, leaveTangent=1.8253, tangentMode=RCTM_User, time=0.3, value=1.002}, <at 1 2 RCIM_Cubic>]}
       ZCurve {keys [<at 0.0788 0>]}
      }}
     CurveIndex:NiagaraFloat=$Particles.NormalizedAge
    ))
   &DynamicMaterialParameters(
    Index 0 Param 4=1
    Module.Param0WriteEnabled=true
    Module.Param1WriteEnabled=true
    Module.Param2WriteEnabled=true
    Module.Param3WriteEnabled=true
   )
   /SetVariables(Particles.Position=&Lerp_Vector(
     A=&TransformPosition(Position=$Particles.LocalPosition, Source Space=::Local, Destination Space=::Simulation)
     B=Linked() {linkedVariable {name=Particles.Position, type=NiagaraPosition}}
     Alpha=&FloatFromCurve(FloatCurve=DataInterface() {propertyValues {Curve {keys [
          {arriveTangent=0.3418, interpMode=RCIM_Cubic, leaveTangent=0.3418, tangentMode=RCTM_User, time=0, value=0}
          {arriveTangent=0.0307, interpMode=RCIM_Cubic, tangentMode=RCTM_Break, time=0.6, value=0.5}
         ]}}})
    )) {isSetParameters=true, module=SetVariables_#5}
   &Color(Color=&ColorFromCurve(
     ColorCurve=DataInterface() {propertyValues {
       AlphaCurve {keys [<at 0 1 RCIM_Cubic>]}
       BlueCurve {keys [<at 0 0 RCIM_Cubic>,<at 0.5 0>,<at 1 1 RCIM_Cubic>]}
       GreenCurve {keys [<at 0 0 RCIM_Cubic>,<at 0.5 1>,<at 1 0 RCIM_Cubic>]}
       RedCurve {keys [<at 0 1 RCIM_Cubic>,<at 0.5 0.2094>,<at 1 0.0831 RCIM_Cubic>]}
      }}
     CurveIndex=&ClampFloat(Float=&Multiply_Float(
       A=&VectorLength(Vector=&Subtract_Vector(A=$Output.SkeletalMeshLocation001.SampledPosition, B=Linked() {linkedVariable {name=Particles.Position, type=NiagaraPosition}}))
       B=0.025
      ))
    )) {enabled=false}
  ]
  rendererClasses [&NiagaraSpriteRendererProperties]
  renderers [{rendererClass=&NiagaraSpriteRendererProperties}]
 }
 snowflake_Local {
  emitterUpdateScript [
   &EmitterState()
   &SpawnRate(SpawnRate=6, Use Spawn Probability=false)
  ]
  particleSpawnScript [
   &InitializeParticle(
    Color=&MakeLinearColorFromVectorAndFloat("Vector (RGB)"=<1 1 1>, "Float (Alpha)"=&RandomRangeFloat(Minimum=1, Maximum=1))
    Lifetime=&RandomRangeFloat(Minimum=0.75, Maximum=2)
    Module.Write Color=true
    Module.Write Lifetime=true
    Module.Write Mass=true
    Module.Write Material Random=true
    Module.Write Position=true
    Module.Write Scale=false
    Module.Write SpriteRotation=true
    Module.Write SpriteSize=true
    Sprite Rotation=&RandomRangeFloat(Minimum=0, Maximum=360)
    Sprite Size=&Vector2DFromFloat(Value=&RandomRangeFloat(Minimum=0.2, Maximum=4))
   )
   &CameraOffset(Camera Offset Amount=2)
   &SkeletalMeshLocation(
    Inherit Velocity=false
    MaskSpawnGroup=false
    Mesh Sampling Type=::NewEnumerator5
    Sampled Position Offset=&RandomRangeVector(Minimum=<-5 0 0>, Maximum=<5 0 0>)
    Sampled Position Offset Coordinate Space:ENiagaraMeshTransforms=::NewEnumerator3
    Skeletal Mesh=$User.SkeletalMesh
    Socket Sampling Mode=::NewEnumerator5
   )
   /SetVariables(Particles.Position=&Add_Vector(
     A=$Output.SkeletalMeshLocation.SampledPosition
     B=&RandomRangeVector(Minimum=<-2 -2 -2>, Maximum=<2 2 2>)
    )) {isSetParameters=true}
  ]
  particleUpdateScript [
   &ParticleState()
   &AddVelocity(Coordinate Space=::World, Velocity=&Multiply_Vector(A=$Particles.Nolmailze, B=<0 1 0>))
   &GravityForce(Gravity=<0 0 -20>)
   &CurlNoiseForce(Noise Frequency=100, Pan Noise=true, Pan Noise Field=<0 0 2>)
   &SolveForcesAndVelocity(Clamp Velocity=false, Limit Acceleration=false)
   &ScaleSpriteSize(Scale Factor=&Vector2DFromFloat(Value=&FloatFromCurve(FloatCurve=DataInterface() {propertyValues {Curve {keys [
          {arriveTangent=4.9073, interpMode=RCIM_Cubic, leaveTangent=4.9073, tangentMode=RCTM_User, time=0, value=0}
          <at 0.2737 1 RCIM_Cubic>
          {arriveTangent=-2.1114, interpMode=RCIM_Cubic, leaveTangent=-2.1114, tangentMode=RCTM_User, time=1, value=0}
         ]}}})))
   &ScaleColor(
    Scale Alpha=&FloatFromCurve(
     FloatCurve=DataInterface() {propertyValues {Curve {keys [
         {arriveTangent=1.2238, interpMode=RCIM_Cubic, leaveTangent=1.2238, tangentMode=RCTM_User, time=0, value=0.3}
         <at 0.4323 0.9 RCIM_Cubic>
         <at 1 0.3 RCIM_Cubic>
        ]}}}
     Scale Curve=0.5
    )
    ScaleA=true
    ScaleRGB=true
   )
  ]
  rendererClasses [&NiagaraSpriteRendererProperties]
  renderers [
   {rendererClass=&NiagaraSpriteRendererProperties}
   {rendererClass=&NiagaraSpriteRendererProperties, rendererIndex=1}
  ]
 }
}
summary {
 systemName=NS_Pickaxe_Frostbyte_Idle
 emitters [
  {emitterName=Frost, rendererClasses [&NiagaraSpriteRendererProperties]}
  {emitterName=snowflake_Local, rendererClasses [&NiagaraSpriteRendererProperties]}
  {emitterName=Frost001, rendererClasses [&NiagaraSpriteRendererProperties]}
  {bEnabled=false, emitterName=Frost002, rendererClasses [&NiagaraSpriteRendererProperties]}
 ]
 userVariables [
  User.bIsFrontEndPreview {type=NiagaraBool}
  User.FrontendSpawn {type=NiagaraBool}
  User.SkeletalMesh {
   type=NiagaraDataInterfaceSkeletalMesh
   defaultValue {
    struct=NiagaraExt_VariableValue_DataInterface
    value {dataInterface=&self:NiagaraDataInterfaceSkeletalMesh_1}
   }
  }
 ]
}
